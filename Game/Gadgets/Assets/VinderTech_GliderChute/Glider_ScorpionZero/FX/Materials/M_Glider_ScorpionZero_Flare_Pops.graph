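# Additive unlit material for Niagara mesh particles, used for flare pop effects on the Scorpion Zero glider.
# The Sprite function billboards meshes using dynamic-parameter scale, and world position offset pulls them toward the camera.
# A borrowed spore texture, rotated by time and per-particle randomness, runs through four-way chaos motion and is masked by a radial falloff.
# Particle color and alpha multiply the result, with depth fade skipped at Low and Medium quality.
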Material /Game/Gadgets/Assets/VinderTech_GliderChute/Glider_ScorpionZero/FX/Materials/M_Glider_ScorpionZero_Flare_Pops
graph {
 properties {
  bUsedWithNiagaraMeshParticles=true
  blendMode=BLEND_Additive
  pixelDepthOffsetMode=PDOM_Legacy
  shadingModel=MSM_Unlit
  usedShadingModels=Unlit
 }
 nodes [
  n1 = WorldPosition()
  n2 = MaterialFunctionCall!(
   Center Location=@n1.XYZ
   0-1 UVs
   XY scale=DynamicParameter() {parameterIndex=1, paramNames [SpriteScale, OpacityMultiplier, "", ""]}.SpriteScale
   OverridePivot
   "Normals (Optional)"
  ) {materialFunction=&Sprite}
  n3 = Subtract(A=CameraPositionWS(), B=Add(A=@n1.XYZ, B=@n2.World position offset))
  Add(A=Multiply(A=Normalize(VectorInput=@n3), B=ScalarParameter() {defaultValue=222, parameterName=Camera Offset}), B=@n2.World position offset)
  n4 = ParticleColor()
  n5 = ScalarParameter() {defaultValue=33, parameterName=Depth Fade}
  n6 = Subtract(A=TextureCoordinate()) {constB=0.5}
  n7 = Add(A=Multiply(A=@n6, B=ScalarParameter() {defaultValue=2.62, parameterName=UV Scale})) {constB=0.5}
  n8 = Constant() {r=2}
  n9 = Add(
   A=Multiply(A=ParticleRandom(), B=MaterialFunctionCall(Multiplier=@n8) {materialFunction=&Pi}.Result)
   B=Time()
  )
  n10 = Power(
   Base=MaterialFunctionCall!(
    Coordinates=Rotator(Coordinate=@n7, Time=@n9) {speed=-1}
    Speed=ScalarParameter() {defaultValue=2.62, parameterName=Speed, sliderMax=9}
    Divisor
    Texture=TextureObject() {samplerType=SAMPLERTYPE_Alpha, texture=/Game/Characters/Player/Male/Medium/Bodies/M_MED_Moss_CarbideBlack/FX/T_Moss_Spores_A}
   ) {materialFunction=/Engine/Functions/Engine_MaterialFunctions03/ComplexMotion/Motion_4WayChaos}.Result
   Exp=ScalarParameter() {defaultValue=3, parameterName=Texture Exponent}
  )
  n11 = Multiply(A=Abs(None=MaterialFunctionCall(Vector 2=@n6) {materialFunction=/Engine/Functions/Engine_MaterialFunctions02/Utility/VectorLength}.V2 Length)) {constB=2}
  n12 = Power(Base=OneMinus(None=Saturate(None=@n11)), Exp=ScalarParameter() {defaultValue=2.6, parameterName=Mask Exponent})
  Multiply(
   A=QualitySwitch(Default=DepthFade(Opacity=@n4.A, FadeDistance=@n5), Low=@n4.A, Medium=@n4.A)
   B=Multiply(A=@n4.RGB, B=Multiply(A=@n10, B=@n12))
  )
 ]
}
